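# Physics asset for the Shark Suit backpack cosmetic, defining collision bodies and joints for its dangling parts.
# The pack bone carries a capsule and box, with three shell capsules and two back-strap boxes constrained directly to it.
# A four-link chain of pouch boxes hangs from the pack, with mostly hinge-like limits that open to a looser tip.
# Straps mainly twist, while shells swing and twist within moderate limits.
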PhysicsAsset /BRCosmetics/Accessories/FORT_Backpacks/Backpack_M_MED_Shark_Suit/Meshes/M_MED_Shark_Suit_Physics
bodies {
 dyn_backstrap_l [{shapeType=Box, extentX=2.9453, extentY=1.01, extentZ=3.2066, center=<0 1.0895 -1.5715>, rotation=<rot 0 0 -40.3966>}]
 dyn_backstrap_r [{shapeType=Box, extentX=2.9453, extentY=1.01, extentZ=3.2066, center=<0 1.0895 -1.5715>, rotation=<rot 0 0 -40.3966>}]
 dyn_pouch_01 [{shapeType=Box, extentX=3.3653, extentY=7.8982, extentZ=0.8239, center=<1.9152 0 -0.3021>}]
 dyn_pouch_02 [{shapeType=Box, extentX=2, extentY=5.5992, extentZ=0.5, center=<1.1571 0 0>}]
 dyn_pouch_03 [{shapeType=Box, extentX=1.1402, extentY=4.4817, extentZ=0.5003, center=<0.6198 0 0>}]
 dyn_pouch_04 [{shapeType=Box, extentX=0.5, extentY=2.5811, extentZ=2, center=<-0.1832 0 -0.1324>, rotation=<rot -58.3467 0 0>}]
 dyn_shell_01 [{shapeType=Capsule, length=2.6626, radius=0.8149, center=<0.0001 0 -0.109>, rotation=<rot 0 -89.9999 -90.0454>}]
 dyn_shell_02 [{shapeType=Capsule, length=2.6626, radius=0.8149, center=<0.0001 0 -0.109>, rotation=<rot 0 -89.9999 -90.0454>}]
 dyn_shell_03 [{shapeType=Capsule, length=2.6626, radius=0.8149, center=<0.0001 0 -0.109>, rotation=<rot 0 -89.9999 -90.0454>}]
 pack [
  {shapeType=Capsule, length=27.5795, radius=12.9676, center=<0 -7.0786 -7.9749>}
  {shapeType=Box, extentX=4.8501, extentY=7.201, extentZ=7.5, center=<-14.1335 -4.593 -13.1425>}
 ]
}
constraints [
 {bone1Name=dyn_shell_01, bone2Name=pack, swing1LimitDegrees=5, swing2LimitDegrees=45, twistLimitDegrees=45}
 {bone1Name=dyn_shell_02, bone2Name=pack, swing1LimitDegrees=5, swing2LimitDegrees=45, twistLimitDegrees=45}
 {bone1Name=dyn_shell_03, bone2Name=pack, swing1LimitDegrees=5, swing2LimitDegrees=45, twistLimitDegrees=45}
 {bone1Name=dyn_pouch_01, bone2Name=pack, swing1LimitDegrees=2, swing2LimitDegrees=30, twistLimitDegrees=2}
 {bone1Name=dyn_pouch_02, bone2Name=dyn_pouch_01, swing1LimitDegrees=3, swing2LimitDegrees=60, twistLimitDegrees=3}
 {bone1Name=dyn_pouch_03, bone2Name=dyn_pouch_02, swing1LimitDegrees=7, swing2LimitDegrees=25, twistLimitDegrees=5}
 {bone1Name=dyn_pouch_04, bone2Name=dyn_pouch_03, swing1LimitDegrees=45, swing2LimitDegrees=45, twistLimitDegrees=45}
 {bone1Name=dyn_backstrap_r, bone2Name=pack, swing1LimitDegrees=2, swing2LimitDegrees=2, twistLimitDegrees=30}
 {bone1Name=dyn_backstrap_l, bone2Name=pack, swing1LimitDegrees=2, swing2LimitDegrees=2, twistLimitDegrees=30}
]
