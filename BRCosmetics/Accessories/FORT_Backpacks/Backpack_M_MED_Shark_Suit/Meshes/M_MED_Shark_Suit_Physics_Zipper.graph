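# Physics asset for the zipper pull on the Shark Suit backpack cosmetic, defining collision and joint limits for its secondary motion.
# The dynamic zipper pull bone has a single box body, while the pack bone carries a capsule and an offset box approximating the bag.
# One constraint joins the pull to the pack with narrow swing limits and a wide twist range, keeping its dangle tightly controlled.
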PhysicsAsset /BRCosmetics/Accessories/FORT_Backpacks/Backpack_M_MED_Shark_Suit/Meshes/M_MED_Shark_Suit_Physics_Zipper
bodies {
 dyn_zipper_pull [{shapeType=Box, extentX=0.5552, extentY=1.7969, extentZ=3.9747, center=<0.0001 0 -1.9406>, rotation=<rot 0 -78.411 0>}]
 pack [
  {shapeType=Capsule, length=27.5795, radius=12.9676, center=<0 -7.0786 -7.9749>}
  {shapeType=Box, extentX=4.8501, extentY=7.201, extentZ=7.5, center=<-14.1335 -4.593 -13.1425>}
 ]
}
constraints [{bone1Name=dyn_zipper_pull, bone2Name=pack, swing1LimitDegrees=5, swing2LimitDegrees=10, twistLimitDegrees=180}]
